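# Sidebar chat list button for one text conversation, showing avatar or party icon, name, time since last activity, message preview or invite, and unread indicator.
# OnListItemObjectSet assigns the conversation viewmodel; helper functions format its data, excluding the local user from avatars.
# Last activity falls back from message time to the later of creation or invite time.
# It extends the sidebar button base, adding compact-UI sizing and name scrolling on hover.
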WidgetBlueprint /Sidebar/Panels/Chat/WBP_Sidebar_ChannelButton
graphs [
 ConvertToSocialUserArrayWithoutSelf {nodes [
   n1 = ConvertToSocialUserArrayWithoutSelf!() -> (then, FortSocialUserVMs:Array of Fort Social User Viewmodel Object References) {is=title K2Node_FunctionEntry}
   > n2 = For Each Loop(Array:Array of Fort Social User Viewmodel Object References=@n1) -> (Array Element:Fort Social User Viewmodel Object Reference)
   Return Node!(
    execute=@n2.Completed
    SocialUsersWithoutSelf:Array of Fort Social User Object References=Get SocialUsers() -> (SocialUsers:Array of Fort Social User Object References)
   )
   n3 = GetSocialUser!(self:Fort Social User Viewmodel Object Reference=@n2) -> (ReturnValue:Fort Social User Object Reference)
   > Branch(Condition=NOT Boolean(A=IsLocalUserForBlueprint!(self:Fort Social User Object Reference=@n3) -> (ReturnValue:bool)))
   > Add(TargetArray:Array of Fort Social User Object References=Get SocialUsers(), NewItem:"Fort Social User Object Reference (by ref)"=@n3) {is=title K2Node_CallArrayFunction}
  ]}
 Convert_StatusIndicatorVisibility {nodes [
   n1 = Convert_StatusIndicatorVisibility!() -> (then, Unread Messages Count:int, HasInvite:bool=~) {is=title K2Node_FunctionEntry}
   n2 = "integer > integer"(A=@n1)
   > Return Node!(
    execute
    Visibility:ESlateVisibility Enum=Select(Option 0=Collapsed, Option 1=SelfHitTestInvisible, Index=OR Boolean(A=@n2, B=@n1.HasInvite))
   )
  ]}
 EventGraph {nodes [
   Event PreConstruct()
   > Parent: PreConstruct(IsDesignTime=false)
   > SetCompactUIStyle(self:WBP Sidebar Button Base Object Reference) {is=title K2Node_CallFunction}
   n1 = Event OnListItemObjectSet()
   > n2 = Cast To FortTextChatConversationVM!(execute, Object=@n1.ListItemObject) -> (then, CastFailed, AsFort Text Chat Conversation VM:Fort Text Chat Conversation VM Object Reference) {is=title K2Node_DynamicCast}
   > SetFortTextChatConversationVM!(execute, self, Viewmodel:Fort Text Chat Conversation VM Object Reference=@n2) -> (then)
  ]}
 GetInvitationText {nodes [
   n1 = GetInvitationText!() -> (then, IsInvitePending:bool=~, InviterUser:Fort Social User Object Reference) {is=title K2Node_FunctionEntry}
   > n2 = Branch(Condition=@n1)
   > Return Node!(
    execute
    InviteText:Text=Format Text!(
     Format="LOCTEXT(\"{name} invited you\")"
     name:String=GetNicknameOrDisplayName!(self:Fort Social User Object Reference=@n1.InviterUser) -> (ReturnValue:String)
    ) -> (Result)
   )
   Return Node(execute=@n2.else)
  ]}
 GetLastActivityTime {nodes [
   n1 = GetLastActivityTime!() -> (then, LastMessage:Text Chat Message VM Object Reference, CreatedTime:Date Time Structure, InvitedTime:Date Time Structure) {is=title K2Node_FunctionEntry}
   > n2 = Branch(Condition=IsValid(Object=@n1))
   > Return Node!(
    execute
    LastActivityTime:Date Time Structure=Get DateReceived!(self:Text Chat Message VM Object Reference=@n1) -> (DateReceived:Date Time Structure)
   )
   Return Node(
    execute=@n2.else
    LastActivityTime=Select(
     Option 0:Date Time Structure=@n1.InvitedTime
     Option 1:Date Time Structure=@n1.CreatedTime
     Index="DateTime > DateTime"!(A:Date Time Structure=@n1.CreatedTime, B:Date Time Structure=@n1.InvitedTime) -> (ReturnValue:bool)
    ) -> (ReturnValue:Date Time Structure)
   )
  ]}
 OnHandleButtonHoverOrUnhovered {nodes [
   n1 = OnHandleButtonHoverOrUnhovered!() -> (then, IsHovered:bool) {is=title K2Node_FunctionEntry}
   > Parent: OnHandleButtonHoverOrUnhovered!(execute, IsHovered:bool=@n1) -> (then) {is=title K2Node_CallParentFunction}
   > SetScrollingEnabled!(execute, self:Common Text Object Reference=Get Text_Name() -> (Text_Name:Common Text Object Reference), bInIsScrollingEnabled:bool=@n1) -> (then)
  ]}
 SetChannelIcon {nodes [
   n1 = SetChannelIcon!() -> (then, ConversationType:EFortTextChatPersistentConversationType Enum=EpicParty) {is=title K2Node_FunctionEntry}
   > SetActiveWidget(
    self=Get Switcher_ChannelIcon() -> (Switcher_ChannelIcon:Common Visibility Switcher Object Reference)
    Widget=Select(
     Option 0:Widget Object Reference=Get PlayerAvatar() -> (PlayerAvatar:WBP Sidebar Player Avatar Object Reference)
     Option 1:Widget Object Reference=Get PartyIcon() -> (PartyIcon:Fort Lazy Image Object Reference)
     Index="Equal (Enum)"(A:EFortTextChatPersistentConversationType Enum=@n1, B:EFortTextChatPersistentConversationType Enum=EpicParty)
    ) -> (ReturnValue:Widget Object Reference)
   )
  ]}
 SetCompactUIStyle {nodes [
   SetCompactUIStyle() {is=title K2Node_FunctionEntry}
   > Parent: SetCompactUIStyle() {is=title K2Node_CallParentFunction}
   > Branch(Condition=IsUsingCompactUI())
   > SetPadding!(
    execute
    self:Horizontal Box Slot Object Reference=SlotAsHorizontalBoxSlot!(Widget:Widget Object Reference=Get VB_Content() -> (VB_Content:Vertical Box Object Reference)) -> (ReturnValue:Horizontal Box Slot Object Reference)
    InPadding_Left:float=0
    InPadding_Top:float=0
    InPadding_Right:float=20
    InPadding_Bottom:float=0
   ) -> (then)
   > SetPadding(
    self=SlotAsHorizontalBoxSlot(Widget=Get Switcher_ChannelIcon())
    InPadding_Left=20
    InPadding_Top=20
    InPadding_Right=24
    InPadding_Bottom=20
   )
   n1 = SlotAsVerticalBoxSlot(Widget=Get HB_BottomLine() -> (HB_BottomLine:Horizontal Box Object Reference))
   n2 = Get!(self:Vertical Box Slot Object Reference=@n1) -> (Padding_Left:float=~, Padding_Top:float=~, Padding_Right:float=~, Padding_Bottom:float=~)
   > SetPadding(
    self:Vertical Box Slot Object Reference=@n1
    InPadding_Left=@n2
    InPadding_Top=4
    InPadding_Right=@n2.Padding_Right
    InPadding_Bottom=@n2.Padding_Bottom
   )
   > Event Set Size!(execute, self:WBP Sidebar Player Avatar Object Reference=Get PlayerAvatar(), NewSize:Vector 2D Structure=(X=88,Y=88)) -> (then)
   > SetDesiredSizeOverride!(execute, self:Image Object Reference=Get PartyIcon(), DesiredSize:Vector 2D Structure=(X=88,Y=88)) -> (then)
  ]}
 SetLastMessagePreviewText {nodes [
   n1 = SetLastMessagePreviewText!() -> (then, LastMessageVM:Text Chat Message VM Object Reference) {is=title K2Node_FunctionEntry}
   > n2 = Is Valid(InputObject=@n1)
   n3 = Get Text_Message() -> (Text_Message:Common Text Object Reference)
   Set Text(execute=@n2.Is Not Valid, self=@n3)
   Set Text(execute=@n2.Is Valid, Text=Get Text(self:Text Chat Message VM Object Reference=@n1), self=@n3)
  ]}
 SetNumUnreadMessages {nodes [
   n1 = SetNumUnreadMessages!() -> (then, NumUnreadMessages:int) {is=title K2Node_FunctionEntry}
   n2 = Get StatusIndicator() -> (StatusIndicator:WBP UIKit Status Indicator Object Reference)
   > Set Configuration!(
    execute
    Configuration:E_UI_StatusIndicator_Configuration Enum=Select(
     Option 0:E_UI_StatusIndicator_Configuration Enum=NewEnumerator0
     Option 1:E_UI_StatusIndicator_Configuration Enum=NewEnumerator3
     Index="integer > integer"(A=@n1)
    ) -> (ReturnValue:E_UI_StatusIndicator_Configuration Enum)
    self:WBP UIKit Status Indicator Base Object Reference=@n2
   ) -> (then, Output_Get:E_UI_StatusIndicator_Configuration Enum=NewEnumerator0)
   > OverrideCountCeiling!(execute, self:WBP UIKit Status Indicator Base Object Reference=@n2, Ceiling:int=9) -> (then)
   > OverrideCount!(execute, self:WBP UIKit Status Indicator Base Object Reference=@n2, Count:int=Get NumUnreadMessages() -> (NumUnreadMessages:int=0)) -> (then)
  ]}
 SetPlayerAvatars {nodes [
   n1 = SetPlayerAvatars!() -> (then, SocialUserVMs:Array of Fort Social User Viewmodel Object References) {is=title K2Node_FunctionEntry}
   > n2 = ConvertToSocialUserArrayWithoutSelf!(execute, self, FortSocialUserVMs:Array of Fort Social User Viewmodel Object References=@n1) -> (then, SocialUsersWithoutSelf:Array of Fort Social User Object References) {is=title K2Node_CallFunction}
   > InitializeWithSocialUsers!(execute, self:WBP Sidebar Player Avatar Object Reference=Get PlayerAvatar(), SocialUsers:Array of Fort Social User Object References=@n2) -> (then)
  ]}
 SetTimespanTextFromDateTime {nodes [
   n1 = SetTimespanTextFromDateTime!() -> (then, LastActivityTime:Date Time Structure) {is=title K2Node_FunctionEntry}
   > "SetText (Text)"(
    self=Get Text_Time() -> (Text_Time:Common Text Object Reference)
    InText=ConvertDateTimeToTimeAgoText!(DateTime:Date Time Structure=@n1) -> (Text:Text)
   )
  ]}
]
variables []
widgets {
 info {
  inheritedWidgetCount=0
  namedSlotCount=1
  parentClass=/Sidebar/Buttons/WBP_Sidebar_Button_Base.WBP_Sidebar_Button_Base_C
  rootWidgetClass=null
  widgetCount=15
 }
 widgets [
  HorizontalBox_2 {widgetClassPath=&HorizontalBox, widget=HorizontalBox_2, properties {bOverride_Cursor=true}}
  Switcher_ChannelIcon {
   widgetClassPath=&CommonVisibilitySwitcher
   bIsVariable=true
   parent=HorizontalBox_2
   widget=Switcher_ChannelIcon
   slot {class=HorizontalBoxSlot, properties {horizontalAlignment=HAlign_Left, verticalAlignment=VAlign_Center, padding=<ltrb 18 0 0 0>}}
  }
  PlayerAvatar {
   widgetClassPath=/Sidebar/Panels/WBP_Sidebar_PlayerAvatar
   bIsVariable=true
   parent=Switcher_ChannelIcon
   widget=PlayerAvatar
   slot {class=CommonVisibilitySwitcherSlot}
  }
  PartyIcon {
   widgetClassPath=&FortLazyImage
   bIsVariable=true
   parent=Switcher_ChannelIcon
   widget=PartyIcon
   properties {brush {
     bIsDynamicallyLoaded=false
     drawAs=Image
     imageType=FullColor
     mirroring=NoMirror
     resourceName=None
     resourceObject=/FortUILibrary/Textures/MaterialInstances/MI_UIKit_Icon_PartyUp
     tiling=NoTile
     imageSize=<60 60>
     margin=<ltrb 0 0 0 0>
     outlineSettings {
      bUseBrushTransparency=false
      roundingType=HalfHeightRadius
      width=0
      color {colorUseRule=UseColor_Specified, specifiedColor=<rgba 0 0 0 0>}
      cornerRadii=<0 0 0 0>
     }
     tintColor {colorUseRule=UseColor_Specified, specifiedColor=<rgba 1 1 1 1>}
     uVRegion {bIsValid=false, max=<0 0>, min=<0 0>}
    }}
   slot {class=CommonVisibilitySwitcherSlot}
  }
  VB_Content {
   widgetClassPath=&VerticalBox
   bIsVariable=true
   parent=HorizontalBox_2
   widget=VB_Content
   slot {class=HorizontalBoxSlot, properties {verticalAlignment=VAlign_Center, padding=<ltrb 18 15 18 15>, size {sizeRule=Fill, value=1}}}
  }
  HorizontalBox_1 {
   widgetClassPath=&HorizontalBox
   parent=VB_Content
   widget=HorizontalBox_1
   slot {class=VerticalBoxSlot, properties {size {sizeRule=Fill, value=1}}}
  }
  Text_Name {
   widgetClassPath=&CommonTextBlock
   bIsVariable=true
   parent=HorizontalBox_1
   widget=Text_Name
   properties {
    bIsScrollingEnabled=false
    clipping=ClipToBounds
    mobileFontSizeMultiplier=1.33
    scrollStyle=&TextScrollStyle-Base_C
    text=Name extra long here really long it needs to scroll on margqueee aflk ajflkjasldf
    textOverflowPolicy=Ellipsis
    colorAndOpacity {colorUseRule=UseColor_Specified, specifiedColor=<rgba 0.7913 0.7913 0.8228 1>}
    font {
     bForceMonospaced=false
     bMaterialIsStencil=false
     fontMaterial=null
     fontObject=&HeadingNow
     letterSpacing=10
     monospacedWidth=1
     size=15.75
     skewAmount=0
     typefaceFontName=75 Medium
     outlineSettings {
      bApplyOutlineToDropShadows=false
      bMiteredCorners=false
      bSeparateFillAlpha=false
      outlineMaterial=/FortUILibrary/Systems/ModularButton/MaterialInstances/MI_UIKit_TextOutline
      outlineSize=0
      outlineColor=<rgba 0 0 0 1>
     }
    }
    shadowOffset=<0 0>
   }
   slot {class=HorizontalBoxSlot, properties {verticalAlignment=VAlign_Center, size {sizeRule=Fill, value=1}}}
  }
  Text_Time {
   widgetClassPath=&CommonTextBlock
   bIsVariable=true
   parent=HorizontalBox_1
   widget=Text_Time
   properties {
    mobileFontSizeMultiplier=1.24
    text=one min ago
    colorAndOpacity {colorUseRule=UseColor_Specified, specifiedColor=<rgba 1 1 1 0.349>}
    font {
     bForceMonospaced=false
     bMaterialIsStencil=false
     fontMaterial=null
     fontObject=&HeadingNow
     letterSpacing=0
     monospacedWidth=1
     size=12.75
     skewAmount=0
     typefaceFontName=74 Regular
     outlineSettings {
      bApplyOutlineToDropShadows=false
      bMiteredCorners=false
      bSeparateFillAlpha=false
      outlineMaterial=&MI_UIKit_TextOutline
      outlineSize=0
      outlineColor=<rgba 0 0 0 1>
     }
    }
    shadowOffset=<0 0>
   }
   slot {class=HorizontalBoxSlot, properties {horizontalAlignment=HAlign_Right, verticalAlignment=VAlign_Center, padding=<ltrb 8 0 0 0>}}
  }
  Spacer_109 {widgetClassPath=&Spacer, parent=HorizontalBox_1, widget=Spacer_109, properties {size=<0 32>}, slot {class=HorizontalBoxSlot}}
  HB_BottomLine {
   widgetClassPath=&HorizontalBox
   bIsVariable=true
   parent=VB_Content
   widget=HB_BottomLine
   slot {class=VerticalBoxSlot, properties {verticalAlignment=VAlign_Center, size {sizeRule=Fill, value=1}}}
  }
  Invite_Message {
   widgetClassPath=&CommonTextBlock
   bIsVariable=true
   parent=HB_BottomLine
   widget=Invite_Message
   properties {
    mobileFontSizeMultiplier=1.41
    text=someonewithlongname invited you
    font {
     bForceMonospaced=false
     bMaterialIsStencil=false
     fontMaterial=/Sidebar/Materials/Fonts/MI_Sidebar_FontColor_Secondary
     fontObject=&HeadingNow
     letterSpacing=0
     monospacedWidth=1
     size=12.75
     skewAmount=0
     typefaceFontName=74 Regular
     outlineSettings {
      bApplyOutlineToDropShadows=false
      bMiteredCorners=false
      bSeparateFillAlpha=false
      outlineMaterial=&MI_UIKit_TextOutline
      outlineSize=0
      outlineColor=<rgba 0 0 0 1>
     }
    }
    shadowOffset=<0 0>
   }
   slot {class=HorizontalBoxSlot, properties {verticalAlignment=VAlign_Center}}
  }
  SizeBox_0 {
   widgetClassPath=&SizeBox
   parent=HB_BottomLine
   widget=SizeBox_0
   properties {bOverride_MaxDesiredHeight=true, maxDesiredHeight=20}
   slot {class=HorizontalBoxSlot, properties {size {sizeRule=Fill, value=1}}}
  }
  Text_Message {
   widgetClassPath=&CommonTextBlock
   bIsVariable=true
   parent=SizeBox_0
   widget=Text_Message
   properties {
    clipping=ClipToBounds
    mobileFontSizeMultiplier=1.41
    text="long text long text long text long text long text long text long text long text long text long text long text long text long text long text "
    textOverflowPolicy=Ellipsis
    colorAndOpacity {colorUseRule=UseColor_Specified, specifiedColor=<rgba 1 1 1 0.651>}
    font {
     bForceMonospaced=false
     bMaterialIsStencil=false
     fontMaterial=null
     fontObject=&HeadingNow
     letterSpacing=0
     monospacedWidth=1
     size=12.75
     skewAmount=0
     typefaceFontName=74 Regular
     outlineSettings {
      bApplyOutlineToDropShadows=false
      bMiteredCorners=false
      bSeparateFillAlpha=false
      outlineMaterial=&MI_UIKit_TextOutline
      outlineSize=0
      outlineColor=<rgba 0 0 0 1>
     }
    }
    shadowOffset=<0 0>
   }
   slot {class=SizeBoxSlot}
  }
  StatusIndicator {
   widgetClassPath=/FortUILibrary/UIKit/StatusIndicator/WBP_UIKit_StatusIndicator
   bIsVariable=true
   parent=HB_BottomLine
   widget=StatusIndicator
   properties {type=Attention, configuration=Number}
   slot {class=HorizontalBoxSlot, properties {horizontalAlignment=HAlign_Right, verticalAlignment=VAlign_Top}}
  }
  Spacer_173 {widgetClassPath=&Spacer, parent=HB_BottomLine, widget=Spacer_173, properties {size=<0 32>}, slot {class=HorizontalBoxSlot}}
 ]
}
