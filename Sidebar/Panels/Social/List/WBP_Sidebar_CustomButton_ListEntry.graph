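# Custom button list entry widget for the Sidebar social panel, derived from WBP_Sidebar_ButtonGeneric with no widgets of its own.
# When assigned a list item, it casts it to BP_Sidebar_CustomButtonEntryVM and stores that view model.
# On click, it runs the parent handler, then broadcasts CustomListButtonClicked with the ButtonType name read from the view model.
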WidgetBlueprint /Sidebar/Panels/Social/List/WBP_Sidebar_CustomButton_ListEntry
graphs [
 CustomListButtonClicked {nodes [CustomListButtonClicked!() -> (then, ButtonType:Name) {is=title K2Node_FunctionEntry}]}
 EventGraph {nodes [
   n1 = Event OnListItemObjectSet()
   > n2 = Cast To BP_Sidebar_CustomButtonEntryVM!(execute, Object=@n1.ListItemObject) -> (then, CastFailed, AsBP Sidebar Custom Button Entry VM:BP Sidebar Custom Button Entry VM Object Reference) {is=title K2Node_DynamicCast}
   > Set BP_Sidebar_CustomButtonEntryVM(BP_Sidebar_CustomButtonEntryVM:BP Sidebar Custom Button Entry VM Object Reference=@n2) -> (Output_Get:BP Sidebar Custom Button Entry VM Object Reference)
   Event On Clicked()
   > Parent: On Clicked() {is=title K2Node_CallParentFunction}
   > Call Custom List Button Clicked!(
    execute:Exec
    self:WBP Sidebar Custom Button List Entry Object Reference
    ButtonType:Name=Get ButtonType!(self:BP Sidebar Custom Button Entry VM Object Reference=Get BP_Sidebar_CustomButtonEntryVM() -> (BP_Sidebar_CustomButtonEntryVM:BP Sidebar Custom Button Entry VM Object Reference)) -> (ButtonType:Name=None)
   ) -> (then:Exec) {is=title K2Node_CallDelegate}
  ]}
]
variables [CustomListButtonClicked {}]
widgets {
 info {
  inheritedWidgetCount=0
  parentClass=/Sidebar/Buttons/WBP_Sidebar_ButtonGeneric.WBP_Sidebar_ButtonGeneric_C
  rootWidgetClass=null
  widgetCount=0
 }
 widgets []
}
